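# Sound cue for the ice cream consumable item's throw-ready sound.
# A Random node picks among three wave variations with equal weights and is excluded from branch culling.
# It plays quietly in the gadget items sound class, using the shared gadget-placement attenuation.
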SoundCue /Game/Athena/Sounds/Items/IceCream/Item_Consumables_IceCream_Throw_Ready_Cue
root=Random_0
nodes [
 {id=Random_0, children [WavePlayer_5,WavePlayer_6,WavePlayer_7], properties {bShouldExcludeFromBranchCulling=true, weights [1,1,1]}}
 {id=WavePlayer_5, properties {soundWaveAssetPtr=&/Item_Consumables_IceCream_Throw_Ready_010}}
 {id=WavePlayer_6, properties {soundWaveAssetPtr=&/Item_Consumables_IceCream_Throw_Ready_020}}
 {id=WavePlayer_7, properties {soundWaveAssetPtr=&/Item_Consumables_IceCream_Throw_Ready_030}}
]
properties {
 attenuationSettings=/Game/Sounds/Attenuation/Fort_GadgetPlace_Att
 maxDistance=2000
 soundClassObject=&Fort_Gadgets_Items
 volumeMultiplier=0.1
}
