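# Sound cue for picking up the Deep Freeze ice cream consumable.
# A Mixer layers the item-specific pickup wave, boosted and slightly delayed, over a quieter generic player melee-equip wave that starts immediately.
# A Modulator at default settings varies only the ice cream layer.
# Playback uses the shared weapon pickup attenuation and loot-collect sound class at reduced overall volume.
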SoundCue /Game/Athena/Sounds/Items/IceCream/Item_Consumables_Icecream_DeepFreeze_Pickup_Cue
root=Mixer_0
nodes [
 {id=Delay_0, children [Modulator_0], properties {delayMax=0.1, delayMin=0.1}}
 {id=Mixer_0, children [Delay_0,WavePlayer_1], properties {inputVolume [1.2,0.5]}}
 {id=Modulator_0, children [WavePlayer_2]}
 {id=WavePlayer_1, properties {soundWaveAssetPtr=/Game/Sounds/Fort_GamePlay_Sounds/Player/player_melee_equip_02}}
 {id=WavePlayer_2, properties {soundWaveAssetPtr=&/Item_Consumables_IceCream_Pickup_DeepFreeze}}
]
properties {
 attenuationSettings=&Fort_Weapon_Pickup_Att
 maxDistance=3500
 soundClassObject=&Fort_Loot_Collect
 volumeMultiplier=0.5
}
